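# Anim Blueprint for the M_MED_HighTower_Wasabi character body that adds procedural muscle deformation to a pose copied from the attached parent mesh.
# Weighted multi-bone constraints place pec aim and right-side measure bones, while Look At and constraint nodes orient pecs, scapulae, lats and traps.
# Stretch and arm-distance alphas drive additive bicep, tricep, pec and right-lat transforms, plus trap translation.
# Those alphas are not declared here, and the EventGraph is empty.
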AnimBlueprint /BRCosmetics/Characters/Player/Male/Medium/Bodies/M_MED_HighTower_Wasabi/Meshes/M_MED_HighTower_Wasabi_AnimBP
target_skeleton=&/M_MED_HighTower_Wasabi_Skeleton
graphs [
 AnimGraph {nodes [
   n1 = Get RightPecStretchAlpha() -> (RightPecStretchAlpha:float)
   "Transform (Modify) Bone - Bone: bicep_02_r"(Alpha=@n1, Translation="0,0,0", Rotation="0,0,0", Scale="0.15,0.15,0.15") {
    rotationMode=BMM_Additive
    rotationSpace=BCS_BoneSpace
    scaleMode=BMM_Additive
    scaleSpace=BCS_BoneSpace
    alphaBoolBlend {blendOption=Linear}
    boneToModify {boneName=bicep_02_r}
    scale=<0.15 0.15 0.15>
   }
   n2 = Get LeftPecStretchAlpha() -> (LeftPecStretchAlpha:float)
   "Transform (Modify) Bone - Bone: bicep_02_l"(Alpha=@n2, Translation="0,0,0", Rotation="0,0,0", Scale="0.15,0.15,0.15") {
    rotationMode=BMM_Additive
    rotationSpace=BCS_BoneSpace
    scaleMode=BMM_Additive
    scaleSpace=BCS_BoneSpace
    alphaBoolBlend {blendOption=Linear}
    boneToModify {boneName=bicep_02_l}
    scale=<0.15 0.15 0.15>
   }
   n3 = "Constraint - pectoralis_02_r_aim (PPPPPPPPPPPPPPPPPP)"!(
    ComponentPose=Local To Component(LocalPose=Copy Pose From Mesh() {bUseAttachedParent=true})
    Alpha=1
    ConstraintWeights_0=0
    ConstraintWeights_1:float=0.0416
    ConstraintWeights_2:float=0.0394
    ConstraintWeights_3:float=0.0003
    ConstraintWeights_4:float=0
    ConstraintWeights_5:float=0
    ConstraintWeights_6:float=0.0122
    ConstraintWeights_7:float=0.259
    ConstraintWeights_8:float=0.0183
    ConstraintWeights_9:float=0.0001
    ConstraintWeights_10:float=0.0106
    ConstraintWeights_11:float=0
    ConstraintWeights_12:float=0.1954
    ConstraintWeights_13:float=0
    ConstraintWeights_14:float=0
    ConstraintWeights_15:float=0
    ConstraintWeights_16:float=0.4228
    ConstraintWeights_17:float=0.0003
   ) -> (Pose) {
    alphaBoolBlend {blendOption=Linear}
    boneToModify {boneName=pectoralis_02_r_aim}
    constraintSetup [
     {targetBone {boneName=tricep_01_r}}
     {targetBone {boneName=upperarm_r}}
     {targetBone {boneName=upperarm_twist_01_r}}
     {targetBone {boneName=deltoid_bk_01_r}}
     {targetBone {boneName=lattissimus_01_r}}
     {targetBone {boneName=pectoralis_02_r}}
     {targetBone {boneName=upperarm_twist_02_r}}
     {targetBone {boneName=deltoid_01_r}}
     {targetBone {boneName=spine_03}}
     {targetBone {boneName=spine_04}}
     {targetBone {boneName=spine_05}}
     {targetBone {boneName=tricep_02_r}}
     {targetBone {boneName=clavicle_r}}
     {targetBone {boneName=neck_02}}
     {targetBone {boneName=neck_01}}
     {targetBone {boneName=bicep_02_r}}
     {targetBone {boneName=deltoid_fr_01_r}}
     {targetBone {boneName=bicep_01_r}}
    ]
    constraintWeights [1,1,1,1,1,1,1,1,1,1,1,1,1,1,1,1,1,1]
   }
   n4 = "Constraint - pectoralis_02_l_aim (PPPPPPPPPPPPPPPPPPPPPPPPPP)"!(
    ComponentPose=@n3
    Alpha=1
    ConstraintWeights_0=0.0003
    ConstraintWeights_1:float=0.0122
    ConstraintWeights_2:float=0
    ConstraintWeights_3:float=0.0183
    ConstraintWeights_4:float=0.0001
    ConstraintWeights_5:float=0.0106
    ConstraintWeights_6:float=0
    ConstraintWeights_7:float=0.4228
    ConstraintWeights_8:float=0.1954
    ConstraintWeights_9:float=0.0394
    ConstraintWeights_10:float=0
    ConstraintWeights_11:float=0
    ConstraintWeights_12:float=0
    ConstraintWeights_13:float=0.0416
    ConstraintWeights_14:float=0
    ConstraintWeights_15:float=0
    ConstraintWeights_16:float=0.259
    ConstraintWeights_17:float=0
    ConstraintWeights_18:float=0
    ConstraintWeights_19:float=0
    ConstraintWeights_20:float=0
    ConstraintWeights_21:float=0.0003
    ConstraintWeights_22:float=0
    ConstraintWeights_23:float=0
    ConstraintWeights_24:float=0
    ConstraintWeights_25:float=0
   ) -> (Pose) {
    alphaBoolBlend {blendOption=Linear}
    boneToModify {boneName=pectoralis_02_l_aim}
    constraintSetup [
     {targetBone {boneName=deltoid_bk_01_l}}
     {targetBone {boneName=upperarm_twist_02_l}}
     {targetBone {boneName=spine_02}}
     {targetBone {boneName=spine_03}}
     {targetBone {boneName=spine_04}}
     {targetBone {boneName=spine_05}}
     {targetBone {boneName=pelvis}}
     {targetBone {boneName=deltoid_fr_01_l}}
     {targetBone {boneName=clavicle_l}}
     {targetBone {boneName=upperarm_twist_01_l}}
     {targetBone {boneName=clavicle_r}}
     {targetBone {boneName=neck_02}}
     {targetBone {boneName=neck_01}}
     {targetBone {boneName=upperarm_l}}
     {targetBone {boneName=lowerarm_twist_02_l}}
     {targetBone {boneName=head}}
     {targetBone {boneName=deltoid_01_l}}
     {targetBone {boneName=lattissimus_01_l}}
     {targetBone {boneName=bicep_02_l}}
     {targetBone {boneName=spine_01}}
     {targetBone {boneName=tricep_02_l}}
     {targetBone {boneName=bicep_01_l}}
     {targetBone {boneName=scapula_01_l}}
     {targetBone {boneName=pectoralis_02_l}}
     {targetBone {boneName=lowerarm_l}}
     {targetBone {boneName=tricep_01_l}}
    ]
    constraintWeights [1,1,1,1,1,1,1,1,1,1,1,1,1,1,1,1,1,1,1,1,1,1,1,1,1,1]
   }
   n5 = Look At - Bone: pectoralis_01_l!(ComponentPose=@n4, Alpha=1, LookAtLocation:Vector="0,0,0") -> (Pose) {
    alphaBoolBlend {blendOption=Linear}
    boneToModify {boneName=pectoralis_01_l}
    lookAtLocation {x=0}
    lookAtTarget {boneReference {boneName=pectoralis_02_l_aim}}
    lookAt_Axis {axis=<1 0>}
    lookUp_Axis {axis {x=0, z=1}}
   }
   n6 = Look At - Bone: pectoralis_01_r!(ComponentPose=@n5, Alpha=1, LookAtLocation:Vector="0,0,0") -> (Pose) {
    alphaBoolBlend {blendOption=Linear}
    boneToModify {boneName=pectoralis_01_r}
    lookAtLocation {x=0}
    lookAtTarget {boneReference {boneName=pectoralis_02_r_aim}}
    lookAt_Axis {axis=<1 0>}
    lookUp_Axis {axis {x=0, z=1}}
   }
   n7 = "Constraint - scapula_base_l (P)"(ComponentPose=@n6, Alpha=1, ConstraintWeights_0=1) {
    alphaBoolBlend {blendOption=Linear}
    boneToModify {boneName=scapula_base_l}
    constraintSetup [{targetBone {boneName=clavicle_l}}]
    constraintWeights [1]
   }
   n8 = Look At - Bone: scapula_00_l(ComponentPose=@n7, Alpha=1) {
    alphaBoolBlend {blendOption=Linear}
    boneToModify {boneName=scapula_00_l}
    lookAtLocation {x=0}
    lookAtTarget {boneReference {boneName=scapula_01_l_aim}}
    lookAt_Axis {axis=<1 0>}
    lookUp_Axis {axis {x=0, z=1}}
   }
   n9 = Look At - Bone: lattissimus_00_l(ComponentPose=@n8, Alpha=1) {
    alphaBoolBlend {blendOption=Linear}
    boneToModify {boneName=lattissimus_00_l}
    lookAtLocation {x=0}
    lookAtTarget {boneReference {boneName=upperarm_l}}
    lookAt_Axis {axis=<1 0>}
    lookUp_Axis {axis {x=0, z=1}}
   }
   n10 = "Constraint - scapula_base_r (P)"(ComponentPose=@n9, Alpha=1, ConstraintWeights_0=1) {
    alphaBoolBlend {blendOption=Linear}
    boneToModify {boneName=scapula_base_r}
    constraintSetup [{targetBone {boneName=clavicle_r}}]
    constraintWeights [1]
   }
   n11 = Look At - Bone: scapula_00_r(ComponentPose=@n10, Alpha=1) {
    alphaBoolBlend {blendOption=Linear}
    boneToModify {boneName=scapula_00_r}
    lookAtLocation {x=0}
    lookAtTarget {boneReference {boneName=scapula_01_r_aim}}
    lookAt_Axis {axis=<1 0>}
    lookUp_Axis {axis {x=0, z=1}}
   }
   n12 = Look At - Bone: lattissimus_00_r(ComponentPose=@n11, Alpha=1) {
    alphaBoolBlend {blendOption=Linear}
    boneToModify {boneName=lattissimus_00_r}
    lookAtLocation {x=0}
    lookAtTarget {boneReference {boneName=upperarm_r}}
    lookAt_Axis {axis=<1 0>}
    lookUp_Axis {axis {x=0, z=1}}
   }
   n13 = Get LeftArmDistanceBICEPAlpha() -> (LeftArmDistanceBICEPAlpha:float)
   n14 = "Transform (Modify) Bone - Bone: bicep_02_l"(ComponentPose=@n12, Alpha=@n13, Translation="-6,-0.000,-0.000", Rotation="0,0,0", Scale="1,1,1") {
    rotationMode=BMM_Additive
    rotationSpace=BCS_BoneSpace
    scaleMode=BMM_Additive
    scaleSpace=BCS_BoneSpace
    translationMode=BMM_Additive
    translationSpace=BCS_BoneSpace
    alphaBoolBlend {blendOption=Linear}
    boneToModify {boneName=bicep_02_l}
    translation {x=-6}
   }
   n15 = "Transform (Modify) Bone - Bone: bicep_01_l"(ComponentPose=@n14, Alpha=@n13, Translation="-0.000,0.000,2.343", Rotation="0,0,0", Scale="1.000,1.500,1.250") {
    rotationMode=BMM_Additive
    rotationSpace=BCS_BoneSpace
    scaleMode=BMM_Additive
    scaleSpace=BCS_BoneSpace
    translationMode=BMM_Additive
    translationSpace=BCS_BoneSpace
    alphaBoolBlend {blendOption=Linear}
    boneToModify {boneName=bicep_01_l}
    scale {y=1.5, z=1.25}
    translation {z=2.343}
   }
   n16 = "Transform (Modify) Bone - Bone: tricep_02_l"(ComponentPose=@n15, Alpha=@n13, Translation="-6,-0.000,-0.000", Rotation="0,0,0", Scale="1,1,1") {
    rotationMode=BMM_Additive
    rotationSpace=BCS_BoneSpace
    scaleMode=BMM_Additive
    scaleSpace=BCS_BoneSpace
    translationMode=BMM_Additive
    translationSpace=BCS_BoneSpace
    alphaBoolBlend {blendOption=Linear}
    boneToModify {boneName=tricep_02_l}
    translation {x=-6}
   }
   n17 = "Transform (Modify) Bone - Bone: tricep_01_l"(ComponentPose=@n16, Alpha=@n13, Translation="-0.000,0.000,2.343", Rotation="0,0,0", Scale="1.000,1.500,1.250") {
    rotationMode=BMM_Additive
    rotationSpace=BCS_BoneSpace
    scaleMode=BMM_Additive
    scaleSpace=BCS_BoneSpace
    translationMode=BMM_Additive
    translationSpace=BCS_BoneSpace
    alphaBoolBlend {blendOption=Linear}
    alphaScaleBiasClamp {scale=0.3}
    boneToModify {boneName=tricep_01_l}
    scale {y=1.5, z=1.25}
    translation {z=2.343}
   }
   n18 = Get RightArmDistanceBICEPAlpha() -> (RightArmDistanceBICEPAlpha:float)
   n19 = "Transform (Modify) Bone - Bone: bicep_02_r"(ComponentPose=@n17, Alpha=@n18, Translation="-6,-0.000,-0.000", Rotation="0,0,0", Scale="1,1,1") {
    rotationMode=BMM_Additive
    rotationSpace=BCS_BoneSpace
    scaleMode=BMM_Additive
    scaleSpace=BCS_BoneSpace
    translationMode=BMM_Additive
    translationSpace=BCS_BoneSpace
    alphaBoolBlend {blendOption=Linear}
    boneToModify {boneName=bicep_02_r}
    translation {x=-6}
   }
   n20 = "Transform (Modify) Bone - Bone: bicep_01_r"(ComponentPose=@n19, Alpha=@n18, Translation="-0.000,0.000,2.343", Rotation="0,0,0", Scale="1.000,1.500,1.250") {
    rotationMode=BMM_Additive
    rotationSpace=BCS_BoneSpace
    scaleMode=BMM_Additive
    scaleSpace=BCS_BoneSpace
    translationMode=BMM_Additive
    translationSpace=BCS_BoneSpace
    alphaBoolBlend {blendOption=Linear}
    boneToModify {boneName=bicep_01_r}
    scale {y=1.5, z=1.25}
    translation {z=2.343}
   }
   n21 = "Transform (Modify) Bone - Bone: tricep_02_r"(ComponentPose=@n20, Alpha=@n18, Translation="-6,-0.000,-0.000", Rotation="0,0,0", Scale="1,1,1") {
    rotationMode=BMM_Additive
    rotationSpace=BCS_BoneSpace
    scaleMode=BMM_Additive
    scaleSpace=BCS_BoneSpace
    translationMode=BMM_Additive
    translationSpace=BCS_BoneSpace
    alphaBoolBlend {blendOption=Linear}
    alphaScaleBiasClamp {scale=0.5}
    boneToModify {boneName=tricep_02_r}
    translation {x=-6}
   }
   n22 = "Transform (Modify) Bone - Bone: tricep_01_r"(ComponentPose=@n21, Alpha=@n18, Translation="-0.000,0.000,2.343", Rotation="0,0,0", Scale="1.000,1.500,1.250") {
    rotationMode=BMM_Additive
    rotationSpace=BCS_BoneSpace
    scaleMode=BMM_Additive
    scaleSpace=BCS_BoneSpace
    translationMode=BMM_Additive
    translationSpace=BCS_BoneSpace
    alphaBoolBlend {blendOption=Linear}
    alphaScaleBiasClamp {scale=0.3}
    boneToModify {boneName=tricep_01_r}
    scale {y=1.5, z=1.25}
    translation {z=2.343}
   }
   n23 = "Transform (Modify) Bone - Bone: bicep_tp_00_r"(ComponentPose=@n22, Alpha=1, Translation="0,0,0", Rotation="0,0,0", Scale="1,1,1") {
    rotationMode=BMM_Additive
    rotationSpace=BCS_BoneSpace
    scaleMode=BMM_Additive
    scaleSpace=BCS_BoneSpace
    alphaBoolBlend {blendOption=Linear}
    boneToModify {boneName=bicep_tp_00_r}
   }
   n24 = "Transform (Modify) Bone - Bone: bicep_bt_00_r"(ComponentPose=@n23, Alpha=1, Translation="0,0,0", Rotation="0.000,0.000,-20", Scale="1,1,1") {
    rotationMode=BMM_Additive
    rotationSpace=BCS_BoneSpace
    alphaBoolBlend {blendOption=Linear}
    boneToModify {boneName=bicep_bt_00_r}
    rotation {roll=-20}
   }
   n25 = "Transform (Modify) Bone - Bone: bicep_01_r"(ComponentPose=@n24, Alpha=@n1, Translation="0,0,0", Rotation="0,0,0", Scale="0.15,0.15,0.15") {
    rotationMode=BMM_Additive
    rotationSpace=BCS_BoneSpace
    scaleMode=BMM_Additive
    scaleSpace=BCS_BoneSpace
    alphaBoolBlend {blendOption=Linear}
    boneToModify {boneName=bicep_01_r}
    scale=<0.15 0.15 0.15>
   }
   n26 = "Transform (Modify) Bone - Bone: pectoralis_02_r"(ComponentPose=@n25, Alpha=@n1, Translation="-8.560,0.000,0", Rotation="0,0,0", Scale="1,1,1") {
    rotationMode=BMM_Additive
    rotationSpace=BCS_BoneSpace
    scaleMode=BMM_Additive
    scaleSpace=BCS_BoneSpace
    translationMode=BMM_Additive
    translationSpace=BCS_BoneSpace
    alphaBoolBlend {blendOption=Linear}
    boneToModify {boneName=pectoralis_02_r}
    translation {x=-8.56}
   }
   n27 = "Transform (Modify) Bone - Bone: bicep_01_l"(ComponentPose=@n26, Alpha=@n2, Translation="0,0,0", Rotation="0,0,0", Scale="0.15,0.15,0.15") {
    rotationMode=BMM_Additive
    rotationSpace=BCS_BoneSpace
    scaleMode=BMM_Additive
    scaleSpace=BCS_BoneSpace
    alphaBoolBlend {blendOption=Linear}
    boneToModify {boneName=bicep_01_l}
    scale=<0.15 0.15 0.15>
   }
   n28 = "Transform (Modify) Bone - Bone: pectoralis_02_l"(ComponentPose=@n27, Alpha=@n2, Translation="-8.560,0.000,-0.000", Rotation="0,0,0", Scale="1,1,1") {
    rotationMode=BMM_Additive
    rotationSpace=BCS_BoneSpace
    scaleMode=BMM_Additive
    scaleSpace=BCS_BoneSpace
    translationMode=BMM_Additive
    translationSpace=BCS_BoneSpace
    alphaBoolBlend {blendOption=Linear}
    boneToModify {boneName=pectoralis_02_l}
    translation {x=-8.5597, z=0}
   }
   n29 = Look At - Bone: dyn_trapezius_01_r(ComponentPose=@n28, Alpha=1) {
    alphaBoolBlend {blendOption=Linear}
    boneToModify {boneName=dyn_trapezius_01_r}
    lookAtLocation {x=0}
    lookAtTarget {boneReference {boneName=dyn_trapezius_02_r_aim}}
    lookAt_Axis {axis=<1 0>}
    lookUp_Axis {axis {x=0, z=1}}
   }
   n30 = Look At - Bone: dyn_trapezius_01_l(ComponentPose=@n29, Alpha=1) {
    alphaBoolBlend {blendOption=Linear}
    boneToModify {boneName=dyn_trapezius_01_l}
    lookAtLocation {x=0}
    lookAtTarget {boneReference {boneName=dyn_trapezius_02_l_aim}}
    lookAt_Axis {axis=<1 0>}
    lookUp_Axis {axis {x=0, z=1}}
   }
   n31 = "Constraint - dyn_trapezius_02_l (T)"(ComponentPose=@n30, Alpha=Get LeftTrapsStretchAlpha() -> (LeftTrapsStretchAlpha:float), ConstraintWeights_0=1) {
    alphaBoolBlend {blendOption=Linear}
    boneToModify {boneName=dyn_trapezius_02_l}
    constraintSetup [{offsetOption=None, transformType=Translation, targetBone {boneName=dyn_trapezius_02_l_aim}}]
    constraintWeights [1]
   }
   n32 = "Constraint - dyn_trapezius_02_r (T)"(ComponentPose=@n31, Alpha=Get RightTrapsStretchAlpha() -> (RightTrapsStretchAlpha:float), ConstraintWeights_0=1) {
    alphaBoolBlend {blendOption=Linear}
    boneToModify {boneName=dyn_trapezius_02_r}
    constraintSetup [{offsetOption=None, transformType=Translation, targetBone {boneName=dyn_trapezius_02_r_aim}}]
    constraintWeights [1]
   }
   n33 = "Constraint - dyn_lat_measure_r (PPPPPPPPPPPPPPPPPPPPPP)"!(
    ComponentPose=@n32
    Alpha=1
    ConstraintWeights_0=0.0002
    ConstraintWeights_1:float=0.0054
    ConstraintWeights_2:float=0.0799
    ConstraintWeights_3:float=0.2789
    ConstraintWeights_4:float=0.2946
    ConstraintWeights_5:float=0.027
    ConstraintWeights_6:float=0.0002
    ConstraintWeights_7:float=0.0176
    ConstraintWeights_8:float=0
    ConstraintWeights_9:float=0.0018
    ConstraintWeights_10:float=0.0039
    ConstraintWeights_11:float=0.0181
    ConstraintWeights_12:float=0
    ConstraintWeights_13:float=0
    ConstraintWeights_14:float=0.2676
    ConstraintWeights_15:float=0.0035
    ConstraintWeights_16:float=0
    ConstraintWeights_17:float=0.0003
    ConstraintWeights_18:float=0
    ConstraintWeights_19:float=0
    ConstraintWeights_20:float=0.001
    ConstraintWeights_21:float=0
   ) -> (Pose) {
    alphaBoolBlend {blendOption=Linear}
    boneToModify {boneName=dyn_lat_measure_r}
    constraintSetup [
     {targetBone {boneName=upperarm_twist_02_r}}
     {targetBone {boneName=spine_01}}
     {targetBone {boneName=spine_02}}
     {targetBone {boneName=spine_03}}
     {targetBone {boneName=spine_04}}
     {targetBone {boneName=spine_05}}
     {targetBone {boneName=pelvis}}
     {targetBone {boneName=ribs_01_r}}
     {targetBone {boneName=deltoid_bk_01_r}}
     {targetBone {boneName=upperarm_r}}
     {targetBone {boneName=upperarm_twist_01_r}}
     {targetBone {boneName=clavicle_r}}
     {targetBone {boneName=neck_02}}
     {targetBone {boneName=deltoid_fr_01_r}}
     {targetBone {boneName=lattissimus_01_r}}
     {targetBone {boneName=tricep_01_r}}
     {targetBone {boneName=lowerarm_r}}
     {targetBone {boneName=scapula_01_r}}
     {targetBone {boneName=pectoralis_02_r}}
     {targetBone {boneName=bicep_01_r}}
     {targetBone {boneName=tricep_02_r}}
     {targetBone {boneName=bicep_02_r}}
    ]
    constraintWeights [1,1,1,1,1,1,1,1,1,1,1,1,1,1,1,1,1,1,1,1,1,1]
   }
   n34 = "Constraint - dyn_tricep_measure_r (PPPPPPPPPPPPPPPPPPPPPPPPP)"!(
    ComponentPose=@n33
    Alpha=1
    ConstraintWeights_0=0.0404
    ConstraintWeights_1:float=0
    ConstraintWeights_2:float=0.0003
    ConstraintWeights_3:float=0.0493
    ConstraintWeights_4:float=0.0182
    ConstraintWeights_5:float=0.0015
    ConstraintWeights_6:float=0.0001
    ConstraintWeights_7:float=0.0017
    ConstraintWeights_8:float=0.0005
    ConstraintWeights_9:float=0
    ConstraintWeights_10:float=0.2736
    ConstraintWeights_11:float=0.1336
    ConstraintWeights_12:float=0.0822
    ConstraintWeights_13:float=0
    ConstraintWeights_14:float=0.0001
    ConstraintWeights_15:float=0
    ConstraintWeights_16:float=0.0098
    ConstraintWeights_17:float=0
    ConstraintWeights_18:float=0.2378
    ConstraintWeights_19:float=0.0008
    ConstraintWeights_20:float=0.0001
    ConstraintWeights_21:float=0.001
    ConstraintWeights_22:float=0.0133
    ConstraintWeights_23:float=0.1321
    ConstraintWeights_24:float=0.0035
   ) -> (Pose) {
    alphaBoolBlend {blendOption=Linear}
    boneToModify {boneName=dyn_tricep_measure_r}
    constraintSetup [
     {targetBone {boneName=upperarm_twist_02_r}}
     {targetBone {boneName=spine_01}}
     {targetBone {boneName=spine_02}}
     {targetBone {boneName=spine_03}}
     {targetBone {boneName=spine_04}}
     {targetBone {boneName=spine_05}}
     {targetBone {boneName=pelvis}}
     {targetBone {boneName=ribs_01_r}}
     {targetBone {boneName=deltoid_bk_01_r}}
     {targetBone {boneName=lowerarm_twist_02_r}}
     {targetBone {boneName=upperarm_r}}
     {targetBone {boneName=upperarm_twist_01_r}}
     {targetBone {boneName=clavicle_r}}
     {targetBone {boneName=neck_02}}
     {targetBone {boneName=deltoid_fr_01_r}}
     {targetBone {boneName=lowerarm_twist_01_r}}
     {targetBone {boneName=lattissimus_01_r}}
     {targetBone {boneName=deltoid_01_r}}
     {targetBone {boneName=tricep_01_r}}
     {targetBone {boneName=lowerarm_r}}
     {targetBone {boneName=scapula_01_r}}
     {targetBone {boneName=pectoralis_02_r}}
     {targetBone {boneName=bicep_01_r}}
     {targetBone {boneName=tricep_02_r}}
     {targetBone {boneName=bicep_02_r}}
    ]
    constraintWeights [1,1,1,1,1,1,1,1,1,1,1,1,1,1,1,1,1,1,1,1,1,1,1,1,1]
   }
   n35 = Get RightLatStretchAlpha() -> (RightLatStretchAlpha:float)
   n36 = "Transform (Modify) Bone - Bone: tricep_01_r"(ComponentPose=@n34, Alpha=@n35, Translation="-0.000,-5.788,-0.000", Rotation="0,0,0", Scale="1.000,0.8,1.000") {
    rotationMode=BMM_Additive
    rotationSpace=BCS_BoneSpace
    scaleMode=BMM_Additive
    scaleSpace=BCS_BoneSpace
    translationMode=BMM_Additive
    translationSpace=BCS_BoneSpace
    alphaBoolBlend {blendOption=Linear}
    boneToModify {boneName=tricep_01_r}
    scale {y=0.8}
    translation {y=-5.788}
   }
   n37 = "Transform (Modify) Bone - Bone: tricep_02_r"(ComponentPose=@n36, Alpha=@n35, Translation="-0.000,-1.148,0.000", Rotation="0,0,0", Scale="1.000,0.8,1.000") {
    rotationMode=BMM_Additive
    rotationSpace=BCS_BoneSpace
    scaleMode=BMM_Additive
    scaleSpace=BCS_BoneSpace
    translationMode=BMM_Additive
    translationSpace=BCS_BoneSpace
    alphaBoolBlend {blendOption=Linear}
    boneToModify {boneName=tricep_02_r}
    scale {y=0.8}
    translation {y=-1.148}
   }
   n38 = "Transform (Modify) Bone - Bone: lattissimus_01_r"(ComponentPose=@n37, Alpha=@n35, Translation="0.000,-7.314,-0.000", Rotation="0,0,0", Scale="1.000,0.000,1.000") {
    rotationMode=BMM_Additive
    rotationSpace=BCS_BoneSpace
    scaleMode=BMM_Additive
    scaleSpace=BCS_BoneSpace
    translationMode=BMM_Additive
    translationSpace=BCS_BoneSpace
    alphaBoolBlend {blendOption=Linear}
    boneToModify {boneName=lattissimus_01_r}
    scale {y=0}
    translation {y=-7.314}
   }
   n39 = "Constraint - lattissimus_01_l (T)"(ComponentPose=@n38, Alpha=1, ConstraintWeights_0=1) {
    alphaBoolBlend {blendOption=Linear}
    boneToModify {boneName=lattissimus_01_l}
    constraintSetup [{transformType=Translation, targetBone {boneName=upperarm_l}}]
    constraintWeights [1]
   }
   n40 = "Constraint - lattissimus_01_r (T)"(ComponentPose=@n39, Alpha=1, ConstraintWeights_0=1) {
    alphaBoolBlend {blendOption=Linear}
    boneToModify {boneName=lattissimus_01_r}
    constraintSetup [{transformType=Translation, targetBone {boneName=upperarm_r}}]
    constraintWeights [1]
   }
   Output Pose(Result=Component To Local(ComponentPose=@n40)) {layerGroup=None}
   Get RightPecAimPosAlpha() -> (RightPecAimPosAlpha:float)
  ]}
 EventGraph {links [], nodes []}
]
variables []
